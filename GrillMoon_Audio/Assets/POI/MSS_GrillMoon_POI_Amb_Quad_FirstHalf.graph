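# MetaSound Source preset that plays the looping quad ambience bed for the first half of the GrillMoon POI lead-up.
# It inherits its graph from MSS_GrillMoon_POI_Amb_Quad, setting the first-half lead-up wave asset and exposing master volume.
# Listener orientation inputs feed the parent, whose internals are not shown here, and output is four-channel quad.
# Close-range POI attenuation, ambience concurrency and the Fort_Events sound class govern playback.
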MetaSoundSource /GrillMoon_Audio/Assets/POI/MSS_GrillMoon_POI_Amb_Quad_FirstHalf
preset_parent=/GrillMoon_Audio/Assets/MSS_GrillMoon_POI_Amb_Quad
inputs [
 MasterVolume {type=Float}
 UE.Listener.Orientation.Azimuth {type=Float}
 UE.Listener.Orientation.Elevation {type=Float}
 UE.Source.OnPlay {type=Trigger}
 Wave Asset {type=WaveAsset}
]
interfaces [UE.Listener.Orientation:1.0, UE.OutputFormat.Quad:1.0, UE.Source:1.1]
outputs [
 "UE.OutputFormat.Quad.Audio:0" {type=Audio}
 "UE.OutputFormat.Quad.Audio:1" {type=Audio}
 "UE.OutputFormat.Quad.Audio:4" {type=Audio}
 "UE.OutputFormat.Quad.Audio:5" {type=Audio}
]
pages [Default {input_defaults {
   MasterVolume=1
   UE.Listener.Orientation.Azimuth=117.031
   UE.Listener.Orientation.Elevation=0
   UE.Source.OnPlay=null
   Wave Asset=/GrillMoon_Audio/Assets/wav/GrillMoon_Amb_Bed_Leadup_FirstHalf_Quad
  }}]
properties {
 assetImportData=&self:AssetImportData
 attenuationSettings=/GrillMoon_Audio/Assets/Settings/ATT_GrillMoon_POI_Close
 bLooping=true
 duration=10000
 maxDistance=14000
 numChannels=4
 outputFormat=Quad
 sampleRate=48000
 soundClassObject=&Fort_Events
 concurrencySet [/GrillMoon_Audio/Assets/Settings/CONC_GrillMoon_Ambience]
}
