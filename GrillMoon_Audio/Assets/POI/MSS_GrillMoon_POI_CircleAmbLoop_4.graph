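# Looping stereo MetaSound for the GrillMoon POI circle ambience, playing the Phase 4 loop wave with portal ambience attenuation and single-instance concurrency.
# A shared BasicLoop stereo MetaSound handles playback using the wave and Master Volume inputs.
# Both channels are multiplied by an interpolated gain that starts from Master Volume.
# The DuckPortalSound trigger sets that gain's target to a fixed level, reached over a short ramp.
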MetaSoundSource /GrillMoon_Audio/Assets/POI/MSS_GrillMoon_POI_CircleAmbLoop_4
inputs [
 DuckPortalSound {type=Trigger}
 Master Volume {type=Float}
 UE.Source.OnPlay {type=Trigger}
 UE.Source.StartTime.StartTime {type=Time}
 Wave Asset {type=WaveAsset}
]
interfaces [UE.OutputFormat.Stereo:1.0, UE.Source.StartTime:1.0, UE.Source:1.1]
outputs [
 "UE.OutputFormat.Stereo.Audio:0" {type=Audio}
 "UE.OutputFormat.Stereo.Audio:1" {type=Audio}
]
pages [Default {
  input_defaults {
   DuckPortalSound=null
   Master Volume=0.3
   UE.Source.OnPlay=null
   UE.Source.StartTime.StartTime=null
   Wave Asset=/GrillMoon_Audio/Assets/wav/GrillMoon_Circle_Amb_Loop_Phase4
  }
  nodes [
   n1 = UE.InterpTo:Audio(Target=Value.Value:Float(Set=^DuckPortalSound, Init Value=^Master Volume) {also {defaults {Target Value=0.35}}}.Output Value) {also {defaults {Interp Time=2}}}
   n2 = /GrillMoon_Audio/Assets/MSS_GrillMoon_BasicLoop_Stereo(
    Master Volume=^Master Volume
    UE.Source.OnPlay=^UE.Source.OnPlay
    UE.Source.StartTime.StartTime=^UE.Source.StartTime.StartTime
    Wave Asset=^Wave Asset
   )
   Multiply_0 = UE.Multiply:Audio by Float(AdditionalOperands=@n1.Value, PrimaryOperand=@n2."UE.OutputFormat.Stereo.Audio:0")
   Multiply_1 = UE.Multiply:Audio by Float(AdditionalOperands=@n1.Value, PrimaryOperand=@n2."UE.OutputFormat.Stereo.Audio:1")
  ]
  links [{dst="@UE.OutputFormat.Stereo.Audio:0", src=Multiply_0.Out}, {dst="@UE.OutputFormat.Stereo.Audio:1", src=Multiply_1.Out}]
 }]
properties {
 assetImportData=&self:AssetImportData
 attenuationSettings=/GrillMoon_Audio/Assets/Settings/ATT_GrillMoon_POI_Portal_Amb
 bLooping=true
 duration=10000
 maxDistance=15900
 numChannels=2
 outputFormat=Stereo
 sampleRate=48000
 soundClassObject=&Fort_Events_Gameplay
 volume=1.4
 concurrencyOverrides {maxCount=1}
 concurrencySet [/GrillMoon_Audio/Assets/Settings/CONC_GrillMoon_Circle]
}
