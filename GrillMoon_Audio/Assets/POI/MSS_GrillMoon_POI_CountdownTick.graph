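# Mono MetaSound for a POI countdown timer, playing a tick on each clock trigger while seconds remain above zero.
# Seconds remaining select the wave from the Sounds array: the regular tick above ten, the final tick at ten and below.
# A trigger sequence alternates between two wave players so successive ticks can overlap, then their mix is scaled by MasterVolume.
# Playback is spatialized with the GrillMoon knockback attenuation and limited to one instance.
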MetaSoundSource /GrillMoon_Audio/Assets/POI/MSS_GrillMoon_POI_CountdownTick
inputs [
 InClockTick {type=Trigger}
 MasterVolume {type=Float}
 SecondsRemaining {type=Int32}
 Sounds {type=WaveAsset:Array}
 UE.Source.OnPlay {type=Trigger}
]
interfaces [UE.OutputFormat.Mono:1.0, UE.Source:1.1]
outputs ["UE.OutputFormat.Mono.Audio:0" {type=Audio}]
pages [Default {
  input_defaults {
   InClockTick=null
   MasterVolume=1.8
   SecondsRemaining=15
   UE.Source.OnPlay=null
   Sounds [/GrillMoon_Audio/Assets/wav/GrillMoon_POI_Timer_Tick, /GrillMoon_Audio/Assets/wav/GrillMoon_POI_Timer_Tick_Final]
  }
  nodes [
   n1 = InitVariable.Trigger()
   n2 = TriggerCompare.Clamp:Int32(A=^SecondsRemaining, Compare=TriggerCompare.Clamp:Int32(Compare=^InClockTick, A=^SecondsRemaining) {also {defaults {B=0, Type=3}}}.True) {also {defaults {B=10, Type=3}}}
   TriggerRouteTriggerRouteInt322_0 = "TriggerRoute.Trigger Route (Int32, 2):Int32"(Set 1=@n2.False, Set 0=@n2.True) {also {defaults {Value 1=1}}}
   TriggerSequence2_0 = "UE.Trigger Sequence (2)"(In=VariableAccessor.Trigger(Variable=VariableMutator.Trigger(Variable=@n1.Variable, Value=@TriggerRouteTriggerRouteInt322_0.On Set).Variable).Value)
   n3 = VariableAccessor.WaveAsset(Variable=VariableMutator.WaveAsset(
     Value=Array.Get:WaveAsset:Array(Array=^Sounds, Trigger=@TriggerRouteTriggerRouteInt322_0.On Set, Index=@TriggerRouteTriggerRouteInt322_0.Value).Element
     Variable=InitVariable.WaveAsset().Variable
    ).Variable)
   WavePlayer_0 = UE.Wave Player:Mono(Play=@TriggerSequence2_0.Out 1, Wave Asset=@n3.Value)
   WavePlayer_1 = UE.Wave Player:Mono(Play=@TriggerSequence2_0.Out 0, Wave Asset=@n3.Value)
   Multiply_0 = UE.Multiply:Audio by Float(
    AdditionalOperands=^MasterVolume
    PrimaryOperand=UE.Add:Audio(AdditionalOperands=@WavePlayer_0.Out Mono, PrimaryOperand=@WavePlayer_1.Out Mono).Out
   )
  ]
  links [{dst="@UE.OutputFormat.Mono.Audio:0", src=Multiply_0.Out}]
 }]
properties {
 assetImportData=&self:AssetImportData
 attenuationSettings=/GrillMoon_Audio/Assets/Settings/ATT_GrillMoon_POI_Knockback
 bLooping=true
 bOverrideConcurrency=true
 duration=10000
 maxDistance=2000
 sampleRate=48000
 soundClassObject=&Fort_Events
 concurrencyOverrides {maxCount=1}
}
