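# Stereo MetaSound Source preset that plays looping point-of-interest music, the Cue2 track, for GrillMoon.
# It derives from the GrillMoon MusicLoop parent, supplying the Cue2 wave asset and a full master volume as inputs, with no graph logic of its own.
# POI music attenuation makes it distance-based, a shared music concurrency set governs simultaneous playback, and it routes to Fortnite's POI music sound class.
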MetaSoundSource /GrillMoon_Audio/Assets/POI/MSS_GrillMoon_POI_MusicLoop_02
preset_parent=/GrillMoon_Audio/Assets/MSS_GrillMoon_MusicLoop
inputs [
 Master Volume {type=Float}
 UE.Source.OnPlay {type=Trigger}
 Wave Asset {type=WaveAsset}
]
interfaces [UE.OutputFormat.Stereo:1.0, UE.Source:1.1]
outputs [
 "UE.OutputFormat.Stereo.Audio:0" {type=Audio}
 "UE.OutputFormat.Stereo.Audio:1" {type=Audio}
]
pages [Default {input_defaults {Master Volume=1, UE.Source.OnPlay=null, Wave Asset=/GrillMoon_Audio/Assets/POI/wav/GrillMoon_MX_POI_Cue2}}]
properties {
 assetImportData=&self:AssetImportData
 attenuationSettings=/GrillMoon_Audio/Assets/Settings/ATT_GrillMoon_POI_Music
 bLooping=true
 duration=10000
 maxDistance=10500
 numChannels=2
 outputFormat=Stereo
 sampleRate=48000
 soundClassObject=/Game/Sounds/FortSoundClassesAndModes/Fort_Music_POI_PSM
 concurrencyOverrides {maxCount=1}
 concurrencySet [/GrillMoon_Audio/Assets/Settings/CONC_GrillMoon_Music]
}
