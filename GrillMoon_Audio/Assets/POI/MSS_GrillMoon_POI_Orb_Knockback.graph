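# MetaSound Source for the knockback sound of the GrillMoon POI orb, played as a stereo one-shot.
# It is a preset of the shared GrillMoon stereo one-shot array MetaSound, supplying three knockback wave variations and a master volume.
# POI knockback attenuation, the Fort_Gadgets sound class and a concurrency override that stops the oldest instance govern its playback.
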MetaSoundSource /GrillMoon_Audio/Assets/POI/MSS_GrillMoon_POI_Orb_Knockback
preset_parent=/GrillMoon_Audio/Assets/MSS_GrillMoon_OneShotArray_Stereo
inputs [
 Master Volume {type=Float}
 Sounds {type=WaveAsset:Array}
 UE.Source.OnPlay {type=Trigger}
 UE.Source.StartTime.StartTime {type=Time}
]
interfaces [UE.OutputFormat.Stereo:1.0, UE.Source.OneShot:1.0, UE.Source.StartTime:1.0, UE.Source:1.1]
outputs [
 "UE.OutputFormat.Stereo.Audio:0" {type=Audio}
 "UE.OutputFormat.Stereo.Audio:1" {type=Audio}
 UE.Source.OneShot.OnFinished {type=Trigger}
]
pages [Default {input_defaults {
   Master Volume=1
   UE.Source.OnPlay=null
   UE.Source.StartTime.StartTime=null
   Sounds [
    /GrillMoon_Audio/Assets/wav/GrillMoon_POI_Orb_Knockback_01
    /GrillMoon_Audio/Assets/wav/GrillMoon_POI_Orb_Knockback_02
    /GrillMoon_Audio/Assets/wav/GrillMoon_POI_Orb_Knockback_03
   ]
  }}]
properties {
 assetImportData=&self:AssetImportData
 attenuationSettings=/GrillMoon_Audio/Assets/Settings/ATT_GrillMoon_POI_Knockback
 bOverrideConcurrency=true
 maxDistance=2000
 numChannels=2
 outputFormat=Stereo
 sampleRate=48000
 soundClassObject=&Fort_Gadgets
 volume=0.6
 concurrencyOverrides {maxCount=3, resolutionRule=StopOldest, retriggerTime=0.3}
}
