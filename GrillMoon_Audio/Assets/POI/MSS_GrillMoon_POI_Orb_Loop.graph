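# MetaSound Source preset that plays the looping stereo ambient sound for GrillMoon's POI orb.
# It inherits its graph from the BasicLoop_Stereo parent, exposing Master Volume and Wave Asset inputs and supplying the orb loop wave.
# A very-close POI attenuation asset handles spatial falloff, and playback routes to the global ambient sound class.
# A concurrency override limits it to a single playing instance.
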MetaSoundSource /GrillMoon_Audio/Assets/POI/MSS_GrillMoon_POI_Orb_Loop
preset_parent=/GrillMoon_Audio/Assets/MSS_GrillMoon_BasicLoop_Stereo
inputs [
 Master Volume {type=Float}
 UE.Source.OnPlay {type=Trigger}
 UE.Source.StartTime.StartTime {type=Time}
 Wave Asset {type=WaveAsset}
]
interfaces [UE.OutputFormat.Stereo:1.0, UE.Source.StartTime:1.0, UE.Source:1.1]
outputs [
 "UE.OutputFormat.Stereo.Audio:0" {type=Audio}
 "UE.OutputFormat.Stereo.Audio:1" {type=Audio}
]
pages [Default {input_defaults {Master Volume=1, UE.Source.OnPlay=null, UE.Source.StartTime.StartTime=null, Wave Asset=/GrillMoon_Audio/Assets/wav/GrillMoon_POI_Orb_Loop}}]
properties {
 assetImportData=&self:AssetImportData
 attenuationSettings=/GrillMoon_Audio/Assets/Settings/ATT_GrillMoon_POI_VeryClose
 bLooping=true
 bOverrideConcurrency=true
 duration=10000
 maxDistance=4100
 numChannels=2
 outputFormat=Stereo
 sampleRate=48000
 soundClassObject=/Game/Sounds/FortSoundClassesAndModes/Fort_Ambient_Global
 volume=1.5
 concurrencyOverrides {maxCount=1}
}
